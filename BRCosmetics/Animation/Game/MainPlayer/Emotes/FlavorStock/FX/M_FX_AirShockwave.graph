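# Unlit translucent, two-sided material for Niagara mesh particles, drawing the air shockwave effect of the FlavorStock emote.
# A turbulence texture, stretched along V and panned by the doubled Time dynamic parameter rather than game time, is masked by vertex red.
# That mask is multiplied by particle alpha and a depth fade, which the quality switch enables only on High and Epic.
# Particle color feeds MF_FX_EmissiveAdjustments for emissive.
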Material /BRCosmetics/Animation/Game/MainPlayer/Emotes/FlavorStock/FX/M_FX_AirShockwave
graph {
 properties {
  bAutomaticallySetUsageInEditor=false
  bUsedWithNiagaraMeshParticles=true
  blendMode=BLEND_Translucent
  shadingModel=MSM_Unlit
  twoSided=true
  usedShadingModels=Unlit
 }
 nodes [
  MaterialFunctionCall(Emissive=ParticleColor().RGB) {materialFunction=&MF_FX_EmissiveAdjustments}
  n1 = Constant() {r=1}
  n2 = DepthFade(FadeDistance=ScalarParameter() {defaultValue=5, parameterName=DepthFadeDistance})
  Multiply(A=Multiply(A=NamedRerouteUsage(), B=QualitySwitch(Default=@n1, Low=@n1, High=@n2, Medium=@n1, Epic=@n2)), B=ParticleColor().A)
  n3 = DynamicParameter() {paramNames [Time,Random,DepthFade,Ref]}
  n4 = Panner(Coordinate=TextureCoordinate() {vTiling=0.1}, Time=Add(A=@n3.Time, B=@n3.Time)) {speedY=0.5}
  NamedRerouteDeclaration(None=Multiply(
    A=TextureSampleParameter2D(UVs=@n4) {parameterName=Wave Texture, texture=/Game/Global/Textures/Default/Caustics/T_Turbulence_Seamless}.R
    B=VertexColor().R
   )) {name=Base, nodeColor=<rgba 0.2422 1 0 1>}
  VertexColor()
 ]
}
