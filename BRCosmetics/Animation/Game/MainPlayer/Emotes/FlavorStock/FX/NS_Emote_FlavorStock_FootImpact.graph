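# Niagara foot-impact effect for the FlavorStock player emote, kicking up a one-shot burst of debris and dust.
# GPU Debris spawns a few randomly sized, tumbling chunks with gravity and collision, varying their renderer visibility tag by quality level or front-end preview.
# Smoke_Radial bursts soft puffs outward that slow under drag and grow, with animated material parameters and time-of-day-scaled alpha.
# A debug gnomon emitter is present but disabled.
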NiagaraSystem /BRCosmetics/Animation/Game/MainPlayer/Emotes/FlavorStock/FX/NS_Emote_FlavorStock_FootImpact
dependencies {
 usedDataInterfaces [
  &NiagaraDataInterfaceCurve
  &NiagaraDataInterfaceColorCurve
  &NiagaraDataInterfaceMeshRendererInfo
  &NiagaraDataInterfacePlatformSet
 ]
 usedDynamicInputs [&UniformRangedVector2D, &MakeFloatFromVector2d]
 usedModules [&SystemState]
}
emitters {
 Debris {
  simTarget=GPUComputeSim
  emitterUpdateScript [
   &EmitterState()
   &SpawnBurst_Instantaneous(
    Spawn Count=&RandomRangeInt(Minimum=3, Maximum=4)
    Spawn Probability=0
    Use Spawn Probability=false
    UseLoopCountLimit=false
   )
  ]
  particleSpawnScript [
   &InitializeParticle(
    Color=<rgba 0.7 0.6723 0.665 1>
    Lifetime=&UniformRangedFloat(Minimum=0.5, Maximum=2)
    Mesh Renderer Array Visibility Mode=::NewEnumerator3
    Mesh Renderer Info=DataInterface() {propertyValues {MeshRenderer=&self:Debris_0.NiagaraMeshRendererProperties_0}}
    Mesh Scale Max=<0.5 0.5 0.5>
    Mesh Scale Min=<0.2 0.2 0.5>
    Mesh Scale Mode=::NewEnumerator2
    Sprite Rotation Mode=::NewEnumerator2
    Sprite Size Max=<3 3>
    Sprite Size Min=<1.5 1.5>
    Sprite Size Mode=::NewEnumerator2
    UsePositionOffset=false
    Write Material Random=true
   ) {module=InitializeParticle001}
   /SetVariables(Particles.RockSize:NiagaraFloat=&UniformAOrBFloat(
     A=&RandomRangeFloat(Minimum=0.4, Maximum=0.45)
     B=0.7
     Distribution Weight=0.2
     Evaluation Type=::NewEnumerator0
     Recalculate Random Each Loop=true
     Override Seed=false
    )) {isSetParameters=true}
   /SetVariables(Particles.SpriteSize=&Multiply_Vector2D_ByFloat(Float=$Particles.RockSize, Vector2D=$Particles.SpriteSize)) {isSetParameters=true}
   /SetVariables(Particles.Scale=&Multiply_VectorByFloat(Vector=$Particles.Scale, Float=$Particles.RockSize)) {isSetParameters=true}
   &ShapeLocation(
    MaskSpawnGroup=false
    Non Uniform Scale=<0.1 1 1>
    Offset=<0 0 5>
    Offset Coordinate Space=::World
    Offset Mode=::NewEnumerator0
    Override Seed=false
    Sphere Radius=7
    UseOwnerScale=true
   ) {module=ShapeLocation002}
   &ShapeLocation(
    Handle Radius=2
    Large Radius=10
    MaskSpawnGroup=false
    Offset Coordinate Space=::Local
    Offset Mode=::NewEnumerator0
    Override Seed=false
    Rotation Mode=::NewEnumerator1
    Shape Primitive=::NewEnumerator3
    UseOwnerScale=true
   ) {enabled=false, module=ShapeLocation001}
   &ShapeLocation(
    Handle Radius=2
    Large Radius=10
    MaskSpawnGroup=false
    Offset Mode=::NewEnumerator0
    Override Seed=false
    Shape Primitive=::NewEnumerator3
    UseOwnerScale=true
   ) {enabled=false}
   &AddVelocityFromPoint(Module.Use Falloff Distance=false, Velocity Strength=&RandomRangeFloat(Minimum=50, Maximum=90))
   &AddVelocity(Velocity=<0 -100 0>, Velocity Speed Scale=&RandomRangeFloat(Minimum=0.1, Maximum=0.5)) {enabled=false, module=AddVelocity001}
   &AddVelocity(Rotation Coordinate Space=::World, Velocity=&RandomRangeVector(Minimum=<0 50 50>, Maximum=<0 0 200>)) {enabled=false}
   &AddRotationalVelocity(Rotation Rate=&RandomVector(Vector Scale=1))
   &PointForce(
    Force Origin=&SimulationPosition()
    Force Strength=&UniformRangedFloat(Minimum=10, Maximum=&Multiply_Float(A=20, B=1))
    Module.Use Falloff=false
    Module.UseLinearForce=false
   ) {enabled=false}
   /SetVariables(Particles.Velocity=&Multiply_VectorByFloat(
     Vector=$Particles.Velocity
     Float=&FloatFromCurve(FloatCurve=DataInterface() {propertyValues {Curve {keys [<at 0.4 2>,<at 1.4 1>]}}}, CurveIndex=$Particles.RockSize)
    )) {isSetParameters=true}
   &ApplyInitialForces()
   /Niagara/Modules/Update/SubUV/SubUVAnimation(Random SubUV Lookup Index:NiagaraFloat=$Particles.RandomFloat, SubUV Animation Mode=::NewEnumerator1) {module=SubUVAnimation001}
   /SetVariables(Particles.VisibilityTag=&MakeCustomIntFromBool(
     Bool=&Boolean_OrOperation(
      Bool1=&IsPlatformSetActive(New Platform Set=DataInterface() {propertyValues {Platforms {qualityLevelMask=-2}}})
      Bool2=$User.bIsFrontEndPreview
     )
     True INT=&RandomRangeInt(Minimum=0, Maximum=1)
     False INT=0
    )) {isSetParameters=true}
  ]
  particleUpdateScript [
   &ParticleState() {module=UpdateAge}
   &Drag(Use Linear Drag=true, Use Rotational Drag=false)
   &GravityForce(Gravity=&Multiply_VectorByFloat(Vector=&RandomRangeVector(Minimum=<0 0 -550>, Maximum=<0 0 -650>), Float=1))
   &Collision(
    Advanced Aging Rate=1
    Age Colliding Particles=false
    Collision Enabled=&SetBoolByFloatComparison(A=$Particles.NormalizedAge, B=0.2)
    Control Roll On Collision=true
    Enable Rest State=false
    GPU Collision Type=::NewEnumerator2
    Module.EnableMaxCollisionCount=false
    Module.Randomize Collision Normal Vector=true
    Particle Radius Scale=0.5
    Restitution=0.4
    ScaleRateOfRotation=false
   )
   &SolveForcesAndVelocity(Clamp Velocity=false, Limit Acceleration=false)
   &SpriteRotationRate(Rotation Rate=&Multiply_Float(A=$Particles.RandomFloat, B=4))
   &ScaleSpriteSize(
    Uniform Curve Scale=2
    Uniform Curve Sprite Scale=DataInterface() {propertyValues {Curve {keys [
        <at 0 0>
        <at 0.1 1>
        {arriveTangent=-1.1111, interpMode=RCIM_Cubic, leaveTangent=-1.1111, time=0.8, value=1}
        {arriveTangent=-5.7307, interpMode=RCIM_Cubic, leaveTangent=-5.7307, tangentMode=RCTM_User, time=1, value=0}
       ]}}}
   )
   &ScaleMeshSize(Scale Factor=&VectorFromFloat(Value=&FloatFromCurve(FloatCurve=DataInterface() {propertyValues {Curve {keys [
          <at 0 0>
          <at 0.1 1>
          {arriveTangent=-1.1111, interpMode=RCIM_Cubic, leaveTangent=-1.1111, time=0.8, value=1}
          {arriveTangent=-5.7307, interpMode=RCIM_Cubic, leaveTangent=-5.7307, tangentMode=RCTM_User, time=1, value=0}
         ]}}})))
   &SpriteRotationRate(Rotation Rate=&RandomRangeFloat(Minimum=-300, Maximum=300)) {module=SpriteRotationRate001}
  ]
  rendererClasses [&NiagaraSpriteRendererProperties, &NiagaraMeshRendererProperties]
  renderers [
   {rendererClass=&NiagaraSpriteRendererProperties}
   {rendererClass=&NiagaraMeshRendererProperties, rendererIndex=1}
   {rendererClass=&NiagaraMeshRendererProperties, rendererIndex=2}
  ]
 }
 FN_DebugGnomon {
  bEnabled=false
  emitterUpdateScript [
   &EmitterState(
    Life Cycle Mode=::NewEnumerator1
    Loop Behavior=::NewEnumerator1
    Loop Duration Mode=::NewEnumerator1
    Scalability Mode=::NewEnumerator1
    UseLoopDelay=false
   )
   &SpawnBurst_Instantaneous(Spawn Count=20, Use Spawn Probability=false, UseLoopCountLimit=false)
  ]
  particleSpawnScript [
   &InitializeParticle(
    Color Mode=::NewEnumerator0
    Lifetime=0.2
    Mesh Scale Mode=::NewEnumerator3
    Mesh Uniform Scale=2
    UsePositionOffset=false
    Write Material Random=true
   )
   &ShapeLocation(
    Handle Radius=2
    Large Radius=10
    MaskSpawnGroup=false
    Offset=<10 0 0>
    Offset Coordinate Space=::Local
    Offset Mode=::NewEnumerator0
    Override Seed=false
    Rotation Mode=::NewEnumerator1
    Shape Primitive=::NewEnumerator3
    UseOwnerScale=true
   )
  ]
  particleUpdateScript [&ParticleState()]
  rendererClasses [&NiagaraMeshRendererProperties]
  renderers [{rendererClass=&NiagaraMeshRendererProperties, rendererIndex=0}]
 }
 Smoke_Radial {
  emitterUpdateScript [
   &EmitterState()
   &SpawnBurst_Instantaneous(Spawn Count=&RandomRangeInt(Minimum=8, Maximum=12), Use Spawn Probability=false) {module=SpawnBurst_Instantaneous001}
  ]
  particleSpawnScript [
   &InitializeParticle(
    Color=<rgba 0.2656 0.2656 0.2656 0.17>
    Lifetime Max=1
    Lifetime Min=0.8
    Lifetime Mode=::NewEnumerator1
    Sprite Rotation Mode=::NewEnumerator2
    Sprite Size Mode=::NewEnumerator1
    Uniform Sprite Size Max=43
    Uniform Sprite Size Min=35
    UsePositionOffset=false
    Write Material Random=true
   ) {module=InitializeParticle001}
   &ShapeLocation(
    Handle Radius=2
    Large Radius=3
    MaskSpawnGroup=false
    Offset=<0 0 5>
    Offset Coordinate Space=::World
    Offset Mode=::NewEnumerator0
    Override Seed=false
    Rotation Angle=90
    Rotation Mode=::NewEnumerator1
    Shape Primitive=::NewEnumerator3
    UseOwnerScale=true
   )
   &AddVelocityFromPoint(
    Module.Use Falloff Distance=false
    Origin Offset=<0 0 5>
    Origin Offset Coordinate Space=::World
    Velocity Origin:Vector3f=&SimulationPosition()
    Velocity Strength=&UniformRangedFloat(Minimum=50, Maximum=200)
   )
   &AddVelocity(Velocity=<0 -200 0>, Velocity Speed Scale=&RandomRangeFloat(Minimum=-0.1, Maximum=0.1)) {module=AddVelocity001}
   &AddVelocity(Velocity=<0 -1 0>, Velocity Speed Scale=&RandomRangeFloat(Minimum=100, Maximum=300)) {enabled=false}
   &ApplyInitialForces()
  ]
  particleUpdateScript [
   &ParticleState()
   &GravityForce(Gravity=<0 0 -200>)
   &Drag(Drag=6, Use Linear Drag=true, Use Rotational Drag=false)
   &SolveForcesAndVelocity(Clamp Velocity=false, Limit Acceleration=false)
   &ScaleSpriteSize(Scale Factor=&Vector2DFromFloat(Value=&FloatFromCurve(FloatCurve=DataInterface() {propertyValues {Curve {keys [
          {arriveTangent=7.1593, interpMode=RCIM_Cubic, leaveTangent=7.1593, tangentMode=RCTM_User, time=0, value=0}
          {arriveTangent=1.1699, interpMode=RCIM_Cubic, leaveTangent=1.1699, tangentMode=RCTM_User, time=0.31, value=0.97}
          <at 1 1.25 RCIM_Cubic>
         ]}}})))
   &DynamicMaterialParameters(
    Index 0 Param 1=&FloatFromCurve(FloatCurve=DataInterface() {propertyValues {Curve {keys [
         {arriveTangent=1.1491, interpMode=RCIM_Cubic, leaveTangent=1.1491, tangentMode=RCTM_User, time=0, value=1}
         {arriveTangent=0.5473, interpMode=RCIM_Cubic, leaveTangent=0.5473, tangentMode=RCTM_User, time=1, value=1.65}
        ]}}})
    Index 2 Param 1=&FloatFromCurve(FloatCurve=DataInterface() {propertyValues {Curve {keys [{interpMode=RCIM_Cubic, leaveTangent=3.2309, tangentMode=RCTM_Break, time=0.1, value=0}, <at 1 1>]}}})
    Index 2 Param 2=&FloatFromCurve(
     FloatCurve=DataInterface() {propertyValues {Curve {keys [<at 0 0.2 RCIM_Cubic>,<at 1 0.8 RCIM_Cubic>]}}}
     Scale Curve=&RandomRangeFloat(Minimum=0.2, Maximum=1.2)
    )
    Index 2 Param 3=&FloatFromCurve(FloatCurve=DataInterface() {propertyValues {Curve {keys [
         {interpMode=RCIM_Cubic, tangentMode=RCTM_Break, time=0, value=1}
         {interpMode=RCIM_Cubic, leaveTangent=11.1233, tangentMode=RCTM_Break, time=1, value=2}
        ]}}})
    Index 2 Param 4=&FloatFromCurve(
     FloatCurve=DataInterface() {propertyValues {Curve {keys [<at 0 1 RCIM_Cubic>, {interpMode=RCIM_Cubic, leaveTangent=-1.2956, tangentMode=RCTM_Break, time=1, value=0.75}]}}}
     Scale Curve=0.5
    )
    Module.Param0WriteEnabled=true
    Module.Param0WriteEnabled2=true
    Module.Param1WriteEnabled=true
    Module.Param1WriteEnabled2=true
    Module.Param2WriteEnabled=false
    Module.Param2WriteEnabled2=true
    Module.Param3WriteEnabled=true
    Module.Param3WriteEnabled2=true
    Write Parameter Index 2=true
   )
   &ScaleColor(
    Scale Alpha=&Multiply_Float(
     A=&FloatFromCurve(FloatCurve=DataInterface() {propertyValues {Curve {keys [<at 0 0>,<at 0.2049 0.9765>,<at 0.7791 0.9107>,<at 1 0>]}}})
     B=&Multiply_Float(A=&FNTimeOfDayScalar(Night Scale=1, Morning Scale=0.3, Evening Scale=0.3), B=1)
    )
    Scale RGB=&VectorFromFloat(Value=&FloatFromCurve(FloatCurve=DataInterface() {propertyValues {Curve {keys [<at 0 1>,<at 0.9983 0.5>]}}}))
    ScaleA=true
    ScaleRGB=true
   )
   &CameraOffset(Camera Offset Amount=15)
  ]
  rendererClasses [&NiagaraSpriteRendererProperties, &NiagaraMeshRendererProperties]
  renderers [
   {rendererClass=&NiagaraSpriteRendererProperties}
   {rendererClass=&NiagaraMeshRendererProperties, rendererIndex=1}
  ]
 }
}
summary {
 systemName=NS_Emote_FlavorStock_FootImpact
 emitters [
  {emitterName=Smoke_Radial, rendererClasses [&NiagaraSpriteRendererProperties, &NiagaraMeshRendererProperties]}
  {
   emitterName=Debris
   simTarget=GPUComputeSim
   rendererClasses [&NiagaraSpriteRendererProperties, &NiagaraMeshRendererProperties]
  }
  {bEnabled=false, emitterName=FN_DebugGnomon, rendererClasses [&NiagaraMeshRendererProperties]}
 ]
 userVariables [
  User.bIsFrontEndPreview {type=NiagaraBool}
  User.SmallerSize {type=NiagaraFloat, defaultValue=0, description=""}
 ]
}
