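# Niagara system for the hand-impact effect of the FlavorStock emote: a burst of dust, rock debris and a mesh shockwave.
# GPU Debris throws sprite or mesh chunks with gravity and quality-gated collision, while Dust and DustLong add puffs driven by curve-fed dynamic material parameters.
# Shock_wave expands a short-lived mesh tinted by time of day, and a Streaks emitter is present but disabled.
# User.SmallerSize scales debris count and dust size, and bIsFrontEndPreview adjusts debris and DustLong color.
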NiagaraSystem /BRCosmetics/Animation/Game/MainPlayer/Emotes/FlavorStock/FX/NS_Emote_FlavorStock_HandImpact
dependencies {
 usedDataInterfaces [
  &NiagaraDataInterfaceCurve
  &NiagaraDataInterfacePlatformSet
  &NiagaraDataInterfaceColorCurve
  &NiagaraDataInterfaceVector2DCurve
  &NiagaraDataInterfaceSpriteRendererInfo
  &NiagaraDataInterfaceMeshRendererInfo
  &NiagaraDataInterfaceVectorCurve
 ]
 usedDynamicInputs [
  &ReturnNormalizedExecIndex
  &UniformRangedVector2D
  &MakeFloatFromVector2d
 ]
 usedModules [
  &SystemState
  &self:SystemSpawnScript.SystemScriptSource.SystemScriptGraph.NiagaraNodeAssignment_0.SetVariables_#1
 ]
}
emitters {
 Debris {
  simTarget=GPUComputeSim
  emitterUpdateScript [
   &EmitterState()
   &SpawnBurst_Instantaneous(
    Spawn Count=&MergeHelper_IntSwitch(
     0=&RandomRangeInt(Minimum=10, Maximum=14)
     1=&RandomRangeInt(Minimum=6, Maximum=8)
     2=?
     3=?
     4=?
     5=?
     Index=&MakeIntFromFloat(Float=$User.SmallerSize)
     Num=::NewEnumerator0
    )
    Spawn Probability=0
    Use Spawn Probability=false
    UseLoopCountLimit=false
   )
  ]
  particleSpawnScript [
   &InitializeParticle(
    Color=<rgba 0.7 0.5459 0.5052 1>
    Lifetime=&UniformRangedFloat(Minimum=0.5, Maximum=2)
    Mesh Renderer Array Visibility Mode=::NewEnumerator3
    Mesh Renderer Info=DataInterface() {propertyValues {MeshRenderer=&self:Debris_0.NiagaraMeshRendererProperties_0}}
    Mesh Scale Max=<0.5 0.5 0.5>
    Mesh Scale Min=<0.2 0.2 0.5>
    Mesh Scale Mode=::NewEnumerator2
    Sprite Rotation Mode=::NewEnumerator2
    Sprite Size Max=<3 3>
    Sprite Size Min=<1.5 1.5>
    Sprite Size Mode=::NewEnumerator2
    UsePositionOffset=false
    Write Material Random=true
   ) {module=InitializeParticle001}
   /SetVariables(Particles.RockSize:NiagaraFloat=&UniformAOrBFloat(
     A=&RandomRangeFloat(Minimum=0.4, Maximum=0.3)
     B=0.7
     Distribution Weight=0.2
     Evaluation Type=::NewEnumerator0
     Recalculate Random Each Loop=true
     Override Seed=false
    )) {isSetParameters=true}
   /SetVariables(Particles.SpriteSize=&Multiply_Vector2D_ByFloat(Float=$Particles.RockSize, Vector2D=$Particles.SpriteSize)) {isSetParameters=true}
   /SetVariables(Particles.Scale=&Multiply_VectorByFloat(Vector=$Particles.Scale, Float=$Particles.RockSize)) {isSetParameters=true}
   &ShapeLocation(MaskSpawnGroup=false, Offset Mode=::NewEnumerator0, Override Seed=false, Sphere Radius=2, UseOwnerScale=true)
   &AddVelocityFromPoint(Module.Use Falloff Distance=false, Velocity Strength=&RandomRangeFloat(Minimum=20, Maximum=80))
   &AddVelocity(Rotation Coordinate Space=::World, Velocity=&RandomRangeVector(Minimum=<0 0 0>, Maximum=<0 0 50>))
   &AddRotationalVelocity(Rotation Rate=&RandomVector(Vector Scale=1))
   &PointForce(
    Force Origin=&SimulationPosition()
    Force Strength=&UniformRangedFloat(Minimum=10, Maximum=&Multiply_Float(A=20, B=1))
    Module.Use Falloff=false
    Module.UseLinearForce=false
   )
   /SetVariables(Particles.Velocity=&Multiply_VectorByFloat(
     Vector=$Particles.Velocity
     Float=&FloatFromCurve(FloatCurve=DataInterface() {propertyValues {Curve {keys [<at 0.4 2>,<at 1.4 1>]}}}, CurveIndex=$Particles.RockSize)
    )) {isSetParameters=true}
   &ApplyInitialForces()
   /Niagara/Modules/Update/SubUV/SubUVAnimation(Random SubUV Lookup Index:NiagaraFloat=$Particles.RandomFloat, SubUV Animation Mode=::NewEnumerator1) {module=SubUVAnimation001}
   /SetVariables(Particles.VisibilityTag=&MakeCustomIntFromBool(
     Bool=&Boolean_OrOperation(
      Bool1=&IsPlatformSetActive(New Platform Set=DataInterface() {propertyValues {Platforms {qualityLevelMask=-2}}})
      Bool2=$User.bIsFrontEndPreview
     )
     True INT=&RandomRangeInt(Minimum=0, Maximum=1)
     False INT=0
    )) {isSetParameters=true}
  ]
  particleUpdateScript [
   &ParticleState() {module=UpdateAge}
   &Drag(Use Linear Drag=true, Use Rotational Drag=false)
   &GravityForce(Gravity=&Multiply_VectorByFloat(Vector=&RandomRangeVector(Minimum=<0 0 -550>, Maximum=<0 0 -650>), Float=1))
   &Collision(
    Advanced Aging Rate=1
    Age Colliding Particles=false
    Collision Enabled=&Boolean_OrOperation(
     Bool1=&IsPlatformSetActive(New Platform Set=DataInterface() {propertyValues {Platforms {qualityLevelMask=-4}}})
     Bool2=$User.bIsFrontEndPreview
    )
    Control Roll On Collision=true
    Enable Rest State=false
    GPU Collision Type=::NewEnumerator2
    Module.EnableMaxCollisionCount=false
    Module.Randomize Collision Normal Vector=true
    Particle Radius Scale=0.5
    Restitution=0.4
    ScaleRateOfRotation=false
   )
   &SolveForcesAndVelocity(Clamp Velocity=false, Limit Acceleration=false)
   &SpriteRotationRate(Rotation Rate=&Multiply_Float(A=$Particles.RandomFloat, B=4))
   &ScaleSpriteSize(
    Uniform Curve Scale=2
    Uniform Curve Sprite Scale=DataInterface() {propertyValues {Curve {keys [
        <at 0 1>
        {arriveTangent=-1, interpMode=RCIM_Cubic, leaveTangent=-1, time=0.8, value=1}
        {arriveTangent=-5.7307, interpMode=RCIM_Cubic, leaveTangent=-5.7307, tangentMode=RCTM_User, time=1, value=0}
       ]}}}
   )
   &ScaleMeshSize(Scale Factor=&VectorFromFloat(Value=&FloatFromCurve(FloatCurve=DataInterface() {propertyValues {Curve {keys [
          <at 0 1>
          {arriveTangent=-1, interpMode=RCIM_Cubic, leaveTangent=-1, time=0.8, value=1}
          {arriveTangent=-5.7307, interpMode=RCIM_Cubic, leaveTangent=-5.7307, tangentMode=RCTM_User, time=1, value=0}
         ]}}})))
   &SpriteRotationRate(Rotation Rate=&RandomRangeFloat(Minimum=-300, Maximum=300)) {module=SpriteRotationRate001}
  ]
  rendererClasses [&NiagaraSpriteRendererProperties, &NiagaraMeshRendererProperties]
  renderers [
   {rendererClass=&NiagaraSpriteRendererProperties}
   {rendererClass=&NiagaraMeshRendererProperties, rendererIndex=1}
  ]
 }
 Dust {
  emitterUpdateScript [
   &EmitterState(Inactive Response=::NewEnumerator1, Life Cycle Mode=::NewEnumerator1, Loop Behavior=::NewEnumerator1, Loop Duration=1, UseLoopDelay=false)
   &SpawnBurst_Instantaneous(Spawn Count=&RandomRangeInt(Minimum=5, Maximum=7), Use Spawn Probability=true, UseLoopCountLimit=false)
  ]
  particleSpawnScript [
   &InitializeParticle(
    Color=&MultiplyLinearColor(
     Linear Color=<rgba 0.222 0.1644 0.111 1>
     Scale Factor=&MakeVector4FromVectorAndFloat("Vector (RGB)"=<1 1 1>, "Float (Alpha)"=&RandomRangeFloat(Minimum=0.5, Maximum=0.77))
    )
    Lifetime Max=0.8
    Lifetime Min=0.5
    Lifetime Mode=::NewEnumerator1
    Sprite Rotation Mode=::NewEnumerator2
    Sprite Size Mode=::NewEnumerator1
    Uniform Sprite Size Max=40
    Uniform Sprite Size Min=20
    UsePositionOffset=true
    Write Material Random=true
   )
   /SetVariables(Particles.InitSpriteRotRate=&UniformRangedFloat(Minimum=-3, Maximum=3)) {isSetParameters=true}
   &ShapeLocation(
    MaskSpawnGroup=false
    Offset Mode=::NewEnumerator0
    Override Seed=false
    Sphere Radius=&Lerp_Float(A=6, B=3, Alpha=$User.SmallerSize)
    UseOwnerScale=true
   )
   &AddVelocity(
    Module.Use Falloff Distance=false
    Origin Offset=<0 0 -5>
    Origin Offset Coordinate Space=::World
    Override Seed=false
    Velocity Mode=::NewEnumerator1
    Velocity Speed=&RandomRangeFloat(Minimum=20, Maximum=50)
   )
   &CameraOffset(Camera Offset Amount=10)
  ]
  particleUpdateScript [
   &ParticleState()
   &SpriteRotationRate(Rotation Rate=&Multiply_Float(
     A=$Particles.InitSpriteRotRate
     B=&FloatFromCurve(
      FloatCurve=DataInterface() {propertyValues {Curve {keys [
          {arriveTangent=-4.0107, interpMode=RCIM_Cubic, leaveTangent=-4.0107, tangentMode=RCTM_User, time=-0.0046, value=1.0113}
          {arriveTangent=-0.2496, interpMode=RCIM_Cubic, leaveTangent=-0.2496, tangentMode=RCTM_User, time=0.5, value=0.16}
         ]}}}
      CurveIndex=$Emitter.Age
      Scale Curve=150
     )
    ))
   &ScaleColor(
    Scale Alpha=&FN_FrontEndFloatModifier(Gameplay=1, Front End=1)
    Scale RGB=&MakeVectorFromLinearColorRGB(Color=<rgba 1 1 1 1>)
    ScaleA=true
    ScaleRGB=true
   ) {module=ScaleColor001}
   &ScaleColor(
    Scale Alpha=&FloatFromCurve(FloatCurve=DataInterface() {propertyValues {Curve {keys [<at 0 0 RCIM_Cubic>,<at 0.1 1 RCIM_Cubic>,<at 1 0 RCIM_Cubic>]}}})
    Scale RGB=&VectorFromFloat(Value=&FloatFromCurve(
      FloatCurve=DataInterface() {propertyValues {Curve {keys [
          {arriveTangent=1.8145, interpMode=RCIM_Cubic, leaveTangent=0.6866, tangentMode=RCTM_Break, time=0, value=0.6}
          {arriveTangent=0.4198, interpMode=RCIM_Cubic, leaveTangent=0.4198, tangentMode=RCTM_User, time=0.4, value=0.825}
          {arriveTangent=0.1918, tangentMode=RCTM_User, time=1, value=1}
         ]}}}
      Scale Curve=&FN_PickFloatDependOnPlatform(TargetPlatformActive=1, TargetPlatformNoActive=0.8)
     ))
    ScaleA=true
    ScaleRGB=true
   )
   /Niagara/Modules/Update/Forces/CurlNoiseForce(Noise Frequency=0.1, Noise Strength=200, Pan Noise=true, Pan Noise Field=<0 0 1>)
   &GravityForce(Gravity=&RandomRangeVector(Minimum=<0 0 -20>, Maximum=<0 0 -250>))
   &Drag(Drag=2, Use Linear Drag=true, Use Rotational Drag=false)
   &InheritSourceMovement(
    Applied Position Scale=&VectorFromFloat(Value=&FloatFromCurve(FloatCurve=DataInterface() {propertyValues {Curve {keys [<at 0.2 1 RCIM_Cubic>,<at 1 0 RCIM_Cubic>]}}}))
    ApplyForce=false
    ApplyPosition=true
    ApplyPositionFalloff=false
   )
   &SolveForcesAndVelocity(Clamp Velocity=false, Limit Acceleration=false)
   &ScaleSpriteSize(Scale Factor=&Vector2DFromFloat(Value=&Lerp_Float(
      A=&FloatFromCurve(
       FloatCurve=DataInterface() {propertyValues {Curve {keys [
           {arriveTangent=1.0025, interpMode=RCIM_Cubic, leaveTangent=1.0025, tangentMode=RCTM_User, time=0, value=0.6}
           {arriveTangent=0.2754, interpMode=RCIM_Cubic, leaveTangent=0.2754, tangentMode=RCTM_User, time=1, value=1.2}
          ]}}}
       Scale Curve=1.1
      )
      B=&FloatFromCurve(
       FloatCurve=DataInterface() {propertyValues {Curve {keys [
           {arriveTangent=1.0025, interpMode=RCIM_Cubic, leaveTangent=1.0025, tangentMode=RCTM_User, time=0, value=0.3}
           {arriveTangent=0.2754, interpMode=RCIM_Cubic, leaveTangent=0.2754, tangentMode=RCTM_User, time=1, value=1.2}
          ]}}}
       Scale Curve=0.9
      )
      Alpha=$User.SmallerSize
     )))
   &DynamicMaterialParameters(
    Index 0 Param 1=&Add_Float(
     A=$Particles.RandomFloat
     B=&FloatFromCurve(
      FloatCurve=DataInterface() {propertyValues {Curve {keys [
          {arriveTangent=2.7533, interpMode=RCIM_Cubic, leaveTangent=1.8958, tangentMode=RCTM_User, time=0, value=1}
          {arriveTangent=0.8292, interpMode=RCIM_Cubic, leaveTangent=0.8292, tangentMode=RCTM_User, time=0.321, value=1.3453}
          {arriveTangent=0.0529, interpMode=RCIM_Cubic, leaveTangent=0.2021, tangentMode=RCTM_User, time=1, value=1.5}
         ]}}}
      CurveIndex=&Add_Float(A=&Multiply_Float(A=$Particles.Initial.MaterialRandom, B=0.5), B=$Particles.NormalizedAge)
     )
    )
    Index 0 Param 2=&FloatFromCurve(FloatCurve=DataInterface() {propertyValues {Curve {keys [
         {arriveTangent=0.0453, interpMode=RCIM_Cubic, tangentMode=RCTM_Break, time=0, value=-1}
         {arriveTangent=0.0604, interpMode=RCIM_Cubic, leaveTangent=0.0604, tangentMode=RCTM_User, time=1, value=-1}
        ]}}})
    Index 0 Param 3=&FloatFromCurve(
     FloatCurve=DataInterface() {propertyValues {Curve {keys [
         {arriveTangent=-1.5217, interpMode=RCIM_Cubic, leaveTangent=-1.5217, tangentMode=RCTM_User, time=0, value=0.2}
         {arriveTangent=0.1017, interpMode=RCIM_Cubic, leaveTangent=0.2723, tangentMode=RCTM_Break, time=0.2, value=0}
        ]}}}
     Scale Curve=0.3
    )
    Index 0 Param 4=&FloatFromCurve(FloatCurve=DataInterface() {propertyValues {Curve {keys [<at 0.5 1>,<at 1 0>]}}})
    Index 1 Param 1=&FloatFromCurve(
     FloatCurve=DataInterface() {propertyValues {Curve {keys [
         {interpMode=RCIM_Cubic, leaveTangent=4.7088, tangentMode=RCTM_Break, time=0, value=0}
         {arriveTangent=4.2257, interpMode=RCIM_Cubic, leaveTangent=4.2257, tangentMode=RCTM_User, time=0.25, value=1}
        ]}}}
     Scale Curve=&RandomRangeFloat(Minimum=1.5, Maximum=1.5)
    )
    Index 1 Param 2=&FloatFromCurve(
     FloatCurve=DataInterface() {propertyValues {Curve {keys [<at 0 0.2>, {leaveTangent=0.9066, time=1, value=2}]}}}
     Scale Curve=&RandomRangeFloat(Minimum=0.2, Maximum=0.3)
    )
    Index 1 Param 3=&FloatFromCurve(FloatCurve=DataInterface() {propertyValues {Curve {keys [
         {interpMode=RCIM_Cubic, leaveTangent=6.3917, tangentMode=RCTM_Break, time=0, value=0.2}
         {arriveTangent=1.7549, interpMode=RCIM_Cubic, leaveTangent=10.6767, tangentMode=RCTM_User, time=0.25, value=1}
        ]}}})
    Index 1 Param 4=&FloatFromCurve(
     FloatCurve=DataInterface() {propertyValues {Curve {keys [{arriveTangent=-0.136, interpMode=RCIM_Cubic, leaveTangent=-0.136, tangentMode=RCTM_User, time=0, value=1}, <at 1 0.95>]}}}
     Scale Curve=&RandomRangeFloat(Minimum=0.5, Maximum=1)
    )
    Index 2 Param 1=&FloatFromCurve(FloatCurve=DataInterface() {propertyValues {Curve {keys [{interpMode=RCIM_Cubic, leaveTangent=1.1368, tangentMode=RCTM_Break, time=0, value=0.45}, <at 1 1>]}}})
    Index 2 Param 2=&FloatFromCurve(
     FloatCurve=DataInterface() {propertyValues {Curve {keys [
         {arriveTangent=1.9056, interpMode=RCIM_Cubic, leaveTangent=1.9056, tangentMode=RCTM_User, time=0, value=0.2}
         {arriveTangent=0.0259, interpMode=RCIM_Cubic, leaveTangent=0.0259, tangentMode=RCTM_User, time=1, value=1.4}
        ]}}}
     Scale Curve=&RandomRangeFloat(Minimum=0.2, Maximum=0.7)
    )
    Index 2 Param 3=&FloatFromCurve(
     FloatCurve=DataInterface() {propertyValues {Curve {keys [
         {interpMode=RCIM_Cubic, leaveTangent=2.3064, tangentMode=RCTM_Break, time=0, value=1}
         {interpMode=RCIM_Cubic, leaveTangent=11.1233, tangentMode=RCTM_Break, time=1, value=2}
        ]}}}
     Scale Curve=1.2
    )
    Index 2 Param 4=&FloatFromCurve(
     FloatCurve=DataInterface() {propertyValues {Curve {keys [
         {arriveTangent=-0.6867, interpMode=RCIM_Cubic, leaveTangent=-0.6867, tangentMode=RCTM_User, time=0, value=1}
         {interpMode=RCIM_Cubic, leaveTangent=-1.2956, tangentMode=RCTM_Break, time=1, value=0.75}
        ]}}}
     Scale Curve=1.2
    )
    Module.Param0WriteEnabled=true
    Module.Param0WriteEnabled1=true
    Module.Param0WriteEnabled2=true
    Module.Param1WriteEnabled=false
    Module.Param1WriteEnabled1=true
    Module.Param1WriteEnabled2=true
    Module.Param2WriteEnabled=true
    Module.Param2WriteEnabled1=true
    Module.Param2WriteEnabled2=true
    Module.Param3WriteEnabled=true
    Module.Param3WriteEnabled1=true
    Module.Param3WriteEnabled2=true
    Write Parameter Index 1=true
    Write Parameter Index 2=true
   )
  ]
  rendererClasses [&NiagaraSpriteRendererProperties]
  renderers [{rendererClass=&NiagaraSpriteRendererProperties}]
 }
 DustLong {
  emitterUpdateScript [
   &EmitterState(Life Cycle Mode=::NewEnumerator1, Loop Behavior=::NewEnumerator1, Loop Delay=0.05, Loop Duration=0.2, UseLoopDelay=false)
   &SpawnBurst_Instantaneous(Spawn Count=2, Spawn Time=0.1, Use Spawn Probability=false, UseLoopCountLimit=false)
  ]
  particleSpawnScript [
   &InitializeParticle(
    Color=&MultiplyLinearColor(
     Linear Color=<rgba 0.222 0.1644 0.111 0.7>
     Scale Factor=&MakeVector4FromVectorAndFloat("Vector (RGB)"=<1 1 1>, "Float (Alpha)"=&RandomRangeFloat(Minimum=0.5, Maximum=0.77))
    )
    Lifetime Max=0.7
    Lifetime Min=0.5
    Lifetime Mode=::NewEnumerator1
    Sprite Size Max=<40 120>
    Sprite Size Min=<30 70>
    Sprite Size Mode=::NewEnumerator2
    UsePositionOffset=false
    Write Material Random=true
   )
   &FN_SetSpawnGroup(Number of Spawn Groups=2)
   &ShapeLocation(
    MaskSpawnGroup=false
    Non Uniform Scale=<3 1 1>
    Offset Coordinate Space=::World
    Offset Mode=::NewEnumerator0
    Override Seed=false
    Sphere Radius=4
    UseOwnerScale=true
   ) {module=ShapeLocation001}
   &SpriteFacingAndAlignment(Sprite Alignment=<0 0 1>, WriteAlignment=true, WriteFacing=false)
   /SetVariables(Particles.unlitColor:LinearColor=<rgba 0.15 0.1339 0.1215 0.25>) {isSetParameters=true}
  ]
  particleUpdateScript [
   &ParticleState()
   &DynamicMaterialParameters(
    Index 0 Param 1=&Add_Float(
     A=0
     B=&FloatFromCurve(
      FloatCurve=DataInterface() {propertyValues {Curve {keys [
          {arriveTangent=2.7533, interpMode=RCIM_Cubic, leaveTangent=1.8958, tangentMode=RCTM_User, time=0, value=1}
          {arriveTangent=0.8292, interpMode=RCIM_Cubic, leaveTangent=0.8292, tangentMode=RCTM_User, time=0.321, value=1.3453}
          {arriveTangent=0.0529, interpMode=RCIM_Cubic, leaveTangent=0.2021, tangentMode=RCTM_User, time=1, value=1.5}
         ]}}}
      CurveIndex=&Add_Float(A=&Multiply_Float(A=$Particles.Initial.MaterialRandom, B=0.5), B=$Particles.NormalizedAge)
     )
    )
    Index 0 Param 2=&FloatFromCurve(FloatCurve=DataInterface() {propertyValues {Curve {keys [
         {arriveTangent=0.0453, interpMode=RCIM_Cubic, tangentMode=RCTM_Break, time=0, value=-1}
         {arriveTangent=0.0604, interpMode=RCIM_Cubic, leaveTangent=0.0604, tangentMode=RCTM_User, time=1, value=-1}
        ]}}})
    Index 0 Param 3=&FloatFromCurve(
     FloatCurve=DataInterface() {propertyValues {Curve {keys [
         {arriveTangent=-1.5217, interpMode=RCIM_Cubic, leaveTangent=-1.5217, tangentMode=RCTM_User, time=0, value=0.2}
         {arriveTangent=0.1017, interpMode=RCIM_Cubic, leaveTangent=0.2723, tangentMode=RCTM_Break, time=0.2, value=0}
        ]}}}
     Scale Curve=0.3
    )
    Index 0 Param 4=&FloatFromCurve(FloatCurve=DataInterface() {propertyValues {Curve {keys [<at 0.5 1>,<at 1 0>]}}})
    Index 1 Param 1=&FloatFromCurve(
     FloatCurve=DataInterface() {propertyValues {Curve {keys [
         {interpMode=RCIM_Cubic, leaveTangent=4.7088, tangentMode=RCTM_Break, time=0, value=0}
         {arriveTangent=4.2257, interpMode=RCIM_Cubic, leaveTangent=4.2257, tangentMode=RCTM_User, time=0.25, value=1}
        ]}}}
     Scale Curve=&RandomRangeFloat(Minimum=1.5, Maximum=1.5)
    )
    Index 1 Param 2=&FloatFromCurve(
     FloatCurve=DataInterface() {propertyValues {Curve {keys [<at 0 0.2>, {leaveTangent=0.9066, time=1, value=2}]}}}
     Scale Curve=&RandomRangeFloat(Minimum=0.2, Maximum=0.3)
    )
    Index 1 Param 3=&FloatFromCurve(FloatCurve=DataInterface() {propertyValues {Curve {keys [
         {interpMode=RCIM_Cubic, leaveTangent=6.3917, tangentMode=RCTM_Break, time=0, value=0.2}
         {arriveTangent=1.7549, interpMode=RCIM_Cubic, leaveTangent=10.6767, tangentMode=RCTM_User, time=0.25, value=1}
        ]}}})
    Index 1 Param 4=&FloatFromCurve(
     FloatCurve=DataInterface() {propertyValues {Curve {keys [{arriveTangent=-0.136, interpMode=RCIM_Cubic, leaveTangent=-0.136, tangentMode=RCTM_User, time=0, value=1}, <at 1 0.95>]}}}
     Scale Curve=&RandomRangeFloat(Minimum=0.5, Maximum=1)
    )
    Index 2 Param 1=&FloatFromCurve(FloatCurve=DataInterface() {propertyValues {Curve {keys [{interpMode=RCIM_Cubic, leaveTangent=1.1368, tangentMode=RCTM_Break, time=0, value=0.45}, <at 1 1>]}}})
    Index 2 Param 2=&FloatFromCurve(
     FloatCurve=DataInterface() {propertyValues {Curve {keys [
         {arriveTangent=1.9056, interpMode=RCIM_Cubic, leaveTangent=1.9056, tangentMode=RCTM_User, time=0, value=0.2}
         {arriveTangent=0.0259, interpMode=RCIM_Cubic, leaveTangent=0.0259, tangentMode=RCTM_User, time=1, value=1.4}
        ]}}}
     Scale Curve=&RandomRangeFloat(Minimum=0.2, Maximum=0.7)
    )
    Index 2 Param 3=&FloatFromCurve(
     FloatCurve=DataInterface() {propertyValues {Curve {keys [
         {interpMode=RCIM_Cubic, leaveTangent=2.3064, tangentMode=RCTM_Break, time=0, value=1}
         {interpMode=RCIM_Cubic, leaveTangent=11.1233, tangentMode=RCTM_Break, time=1, value=2}
        ]}}}
     Scale Curve=1.2
    )
    Index 2 Param 4=&FloatFromCurve(
     FloatCurve=DataInterface() {propertyValues {Curve {keys [
         {arriveTangent=-0.6867, interpMode=RCIM_Cubic, leaveTangent=-0.6867, tangentMode=RCTM_User, time=0, value=1}
         {interpMode=RCIM_Cubic, leaveTangent=-1.2956, tangentMode=RCTM_Break, time=1, value=0.75}
        ]}}}
     Scale Curve=1.2
    )
    Module.Param0WriteEnabled=true
    Module.Param0WriteEnabled1=true
    Module.Param0WriteEnabled2=true
    Module.Param1WriteEnabled=false
    Module.Param1WriteEnabled1=true
    Module.Param1WriteEnabled2=true
    Module.Param2WriteEnabled=true
    Module.Param2WriteEnabled1=true
    Module.Param2WriteEnabled2=true
    Module.Param3WriteEnabled=true
    Module.Param3WriteEnabled1=true
    Module.Param3WriteEnabled2=true
    Write Parameter Index 1=true
    Write Parameter Index 2=true
   ) {module=DynamicMaterialParameters001}
   &GravityForce(Gravity=&RandomRangeVector(Minimum=<0 0 -1>, Maximum=<0 0 -10>))
   &InheritSourceMovement(ApplyForce=false, ApplyPosition=true, ApplyPositionFalloff=false)
   &SolveForcesAndVelocity(Clamp Velocity=false, Limit Acceleration=false)
   &ScaleSpriteSize(
    Non-Uniform Curve Scale=&Vector2DFromFloat(Value=&Multiply_Float(A=0.5, B=&RandomRangeFloat(Minimum=0.5, Maximum=1)))
    Non-Uniform Curve Sprite Scale=DataInterface() {propertyValues {
      XCurve {keys [{tangentMode=RCTM_User, time=0.5, value=1}, {tangentMode=RCTM_User, time=1, value=1}]}
      YCurve {keys [
        {arriveTangent=2.1757, interpMode=RCIM_Cubic, leaveTangent=2.1757, tangentMode=RCTM_User, time=0, value=0.2}
        {interpMode=RCIM_Cubic, tangentMode=RCTM_User, time=1, value=1}
       ]}
     }}
    Scale Sprite Size Mode=::NewEnumerator3
   )
   &CameraOffset(Camera Offset Amount=&RandomRangeFloat(Minimum=15, Maximum=20))
   &ScaleColor(
    Scale Alpha=&FN_FrontEndFloatModifier(Gameplay=0.7, Front End=1)
    Scale RGB=&MakeVectorFromLinearColorRGB(Color=&MakeCustomLinearColorFromBool(Bool=$User.bIsFrontEndPreview, True Color=<rgba 0.5 0.5 0.5 1>, False Color=<rgba 1 1 1 1>))
    ScaleA=true
    ScaleRGB=true
   ) {module=ScaleColor001}
   &ScaleColor(
    Scale Alpha=&FloatFromCurve(FloatCurve=DataInterface() {propertyValues {Curve {keys [<at 0 0 RCIM_Cubic>,<at 0.1 1 RCIM_Cubic>,<at 1 0 RCIM_Cubic>]}}})
    Scale RGB=&VectorFromFloat(Value=&FloatFromCurve(
      FloatCurve=DataInterface() {propertyValues {Curve {keys [
          {arriveTangent=1.8145, interpMode=RCIM_Cubic, leaveTangent=0.8058, tangentMode=RCTM_Break, time=0, value=0.5}
          {arriveTangent=0.5068, interpMode=RCIM_Cubic, leaveTangent=0.5068, tangentMode=RCTM_User, time=0.4, value=0.825}
          {arriveTangent=0.1918, tangentMode=RCTM_User, time=1, value=1}
         ]}}}
      Scale Curve=&FN_PickFloatDependOnPlatform(TargetPlatformActive=1, TargetPlatformNoActive=0.8)
     ))
    ScaleA=true
    ScaleRGB=true
   ) {module=ScaleColor002}
  ]
  rendererClasses [&NiagaraSpriteRendererProperties]
  renderers [{rendererClass=&NiagaraSpriteRendererProperties}]
 }
 Shock_wave {
  emitterUpdateScript [
   &EmitterState(
    Inactive Response=::NewEnumerator1
    Life Cycle Mode=::NewEnumerator1
    Loop Behavior=::NewEnumerator1
    Loop Delay=0.2
    Loop Duration=3
    UseLoopDelay=false
   )
   &SpawnBurst_Instantaneous(Spawn Count=2, Use Spawn Probability=false, UseLoopCountLimit=false)
  ]
  particleSpawnScript [
   &InitializeParticle(
    Color=&RandomRangeLinearColor(Minimum=<rgba 0.12 0.0982 0.078 1.3>, Maximum=<rgba 0.2 0.1606 0.124 1.3>)
    Lifetime=0.3
    Mesh Scale Mode=::NewEnumerator3
    Mesh Uniform Scale=&Multiply_Float(A=&RandomRangeFloat(Minimum=0.4, Maximum=0.6), B=0.7)
    UsePositionOffset=false
    Write Material Random=true
   ) {module=InitializeParticle001}
   &InitialMeshOrientation(
    Mesh Orientation Mode=::NewEnumerator7
    Rotation=&RandomRangeVector(Minimum=<0 0 0>, Maximum=<1 0 0>)
    Use Rotation Vector=true
   )
  ]
  particleUpdateScript [
   &ParticleState()
   &SolveForcesAndVelocity(Clamp Velocity=false, Limit Acceleration=false)
   &ScaleColor(
    Scale Alpha=&Multiply_Float(
     A=&FN_FrontEndFloatModifier(Gameplay=1, Front End=1)
     B=&Multiply_Float(
      A=&FloatFromCurve(FloatCurve=DataInterface() {propertyValues {Curve {keys [<at 0 0 RCIM_Cubic>,<at 0.026 1 RCIM_Cubic>,<at 1 0>]}}})
      B=&Multiply_Float(
       A=1.7
       B=&FNTimeOfDayScalar(Night Scale=1.1, Morning Scale=1.3, Day Scale=1.333, Evening Scale=1.3)
      )
     )
    )
    Scale RGB=&MakeVectorFromLinearColorRGB(Color=&FNTimeOfDayColor(
      Night Color=<rgba 0.2167 0.3012 0.4 1>
      Morning Color=<rgba 0.4 0.2668 0.2865 1>
      Day Color=<rgba 1.05 1.05 1.05 1>
      Evening Color=<rgba 0.4 0.2668 0.2865 1>
     ))
    ScaleA=true
    ScaleRGB=true
   ) {module=ScaleColor001}
   &ScaleMeshSize(Scale Factor=&VectorFromCurve(
     VectorCurve=DataInterface() {propertyValues {XCurve {keys [<at 0 0.3>]}, YCurve {keys [<at 0 0.3>,<at 1 2>]}, ZCurve {keys [<at 0 0.3>,<at 1 2>]}}}
     Scale Curve=<0.01 1 1>
    ))
   &DynamicMaterialParameters(
    Index 0 Param 1=&FloatFromCurve(FloatCurve=DataInterface() {propertyValues {Curve {keys [<at 0 0 RCIM_Cubic>,<at 1 0.28 RCIM_Cubic>]}}})
    Index 0 Param 2=&RandomRangeFloat(Minimum=0, Maximum=1)
    Index 0 Param 3=100
    Index 0 Param 4=&FloatFromCurve(FloatCurve=DataInterface() {propertyValues {Curve {keys [<at 0 0>, {arriveTangent=-3.0374, leaveTangent=2.0371, time=0.2, value=0}, {interpMode=RCIM_Cubic, tangentMode=RCTM_Break, time=1, value=0}]}}})
    Module.Param0WriteEnabled=true
    Module.Param1WriteEnabled=true
    Module.Param2WriteEnabled=true
    Module.Param3WriteEnabled=true
   )
  ]
  rendererClasses [&NiagaraMeshRendererProperties]
  renderers [{rendererClass=&NiagaraMeshRendererProperties, rendererIndex=0}]
 }
 Streaks {
  bEnabled=false
  emitterUpdateScript [
   &EmitterState()
   &SpawnBurst_Instantaneous(Spawn Count=4, Use Spawn Probability=false, UseLoopCountLimit=false)
  ]
  particleSpawnScript [
   &InitializeParticle(
    Color=&RandomRangeLinearColor(Minimum=<rgba 0.3 0.2221 0.15 0.9>, Maximum=<rgba 0.35 0.241 0.14 1>)
    Lifetime Max=0.2
    Lifetime Min=0.1
    Lifetime Mode=::NewEnumerator1
    Sprite Size=&Multiply_Vector2D_ByFloat(Float=0.7, Vector2D=&RandomRangeVector2D(Minimum=<10 70>, Maximum=<15 120>))
    Sprite Size Mode=::NewEnumerator4
    Sprite UV Mode=::NewEnumerator1
    UsePositionOffset=false
    Write Material Random=true
   )
   &FN_SetSpawnGroup(Number of Spawn Groups=2)
   &ShapeLocation(
    MaskSpawnGroup=false
    Offset Mode=::NewEnumerator0
    Override Seed=false
    Shape Origin=$Particles.Initial.Position
    Sphere Radius=5
    UseOwnerScale=true
   )
   &AddVelocity(Module.Use Falloff Distance=false, Override Seed=false, Velocity Mode=::NewEnumerator1, Velocity Speed=1)
  ]
  particleUpdateScript [
   &ParticleState()
   &InheritSourceMovement(ApplyForce=false, ApplyPosition=true, ApplyPositionFalloff=false)
   &SolveForcesAndVelocity(Clamp Velocity=false, Limit Acceleration=false)
   &ScaleColor(
    Scale Alpha=&FloatFromCurve(
     FloatCurve=DataInterface() {propertyValues {Curve {keys [
         {arriveTangent=0.0001, interpMode=RCIM_Cubic, leaveTangent=0.0001, tangentMode=RCTM_User, time=0.25, value=1}
         {arriveTangent=-0.8752, interpMode=RCIM_Cubic, leaveTangent=-0.8752, tangentMode=RCTM_User, time=1, value=0}
        ]}}}
     Scale Curve=&Multiply_Float(
      A=&Multiply_Float(A=0.666, B=&Lerp_Float(A=0.85, B=1, Alpha=$User.SmallerSize))
      B=&FNTimeOfDayScalar(Night Scale=0.8, Morning Scale=1, Evening Scale=1)
     )
    )
    Scale RGB=&Multiply_Vector(
     A=&MakeVectorFromLinearColorRGB(Color=&ColorFromCurve(ColorCurve=DataInterface() {propertyValues {
         AlphaCurve {keys [<at 0 1 RCIM_Cubic>,<at 1 0 RCIM_Cubic>]}
         BlueCurve {keys [{arriveTangent=2.4053, interpMode=RCIM_Cubic, time=0.2209, value=1.5}, <at 0.6132 0.7 RCIM_Cubic>]}
         GreenCurve {keys [{arriveTangent=2.4053, interpMode=RCIM_Cubic, time=0.2209, value=1.5}, <at 0.6132 0.7 RCIM_Cubic>]}
         RedCurve {keys [{arriveTangent=2.4053, interpMode=RCIM_Cubic, time=0.2209, value=1.5}, <at 0.6132 0.7 RCIM_Cubic>]}
        }}))
     B=&MakeVectorFromLinearColorRGB(Color=&FNTimeOfDayColor(
       Night Color=<rgba 0.2167 0.3012 0.4 1>
       Morning Color=<rgba 0.4 0.2668 0.2865 1>
       Day Color=<rgba 1 1 1 1>
       Evening Color=<rgba 0.4 0.2668 0.2865 1>
      ))
    )
    ScaleA=true
    ScaleRGB=true
   )
   &ScaleSpriteSize(
    Non-Uniform Curve Scale=&Vector2DFromFloat(Value=&Lerp_Float(A=0.8, B=0.6, Alpha=$User.SmallerSize))
    Non-Uniform Curve Sprite Scale=DataInterface() {propertyValues {
      XCurve {keys [
        {arriveTangent=3.2597, interpMode=RCIM_Cubic, leaveTangent=3.2597, tangentMode=RCTM_User, time=0, value=0}
        {interpMode=RCIM_Cubic, tangentMode=RCTM_User, time=1, value=1}
       ]}
      YCurve {keys [
        {arriveTangent=1.3826, interpMode=RCIM_Cubic, leaveTangent=1.3826, tangentMode=RCTM_User, time=0, value=0.2}
        {interpMode=RCIM_Cubic, tangentMode=RCTM_User, time=1, value=1}
       ]}
     }}
    Scale Sprite Size Mode=::NewEnumerator3
   )
   &DynamicMaterialParameters(
    Index 0 Param 2=1
    Module.Param0WriteEnabled=true
    Module.Param0WriteEnabled2=true
    Module.Param1WriteEnabled=true
    Module.Param1WriteEnabled2=true
    Module.Param2WriteEnabled=true
    Module.Param2WriteEnabled2=true
    Module.Param3WriteEnabled=true
    Module.Param3WriteEnabled2=true
    Write Parameter Index 2=true
   )
   &CameraOffset(Camera Offset Amount=&RandomRangeFloat(Minimum=10, Maximum=35))
   /Niagara/Modules/Update/SubUV/V2/SubUVAnimation(
    Evaluation Type=::NewEnumerator0
    "Sprite Renderer "=DataInterface() {propertyValues {SpriteRenderer=&self:Streaks_0.NiagaraSpriteRendererProperties_2}}
    SubUV Animation Mode:ENiagara_SubUVLookupModeV2=::NewEnumerator2
    UseEndFrame=false
    UseStartFrame=false
   )
  ]
  rendererClasses [&NiagaraSpriteRendererProperties]
  renderers [{rendererClass=&NiagaraSpriteRendererProperties}]
 }
}
summary {
 systemName=NS_Emote_FlavorStock_HandImpact
 emitters [
  {emitterName=Dust, rendererClasses [&NiagaraSpriteRendererProperties]}
  {bEnabled=false, emitterName=Streaks, rendererClasses [&NiagaraSpriteRendererProperties]}
  {
   emitterName=Debris
   simTarget=GPUComputeSim
   rendererClasses [&NiagaraSpriteRendererProperties, &NiagaraMeshRendererProperties]
  }
  {emitterName=DustLong, rendererClasses [&NiagaraSpriteRendererProperties]}
  {emitterName=Shock_wave, rendererClasses [&NiagaraMeshRendererProperties]}
 ]
 userVariables [
  User.bIsFrontEnd {type=NiagaraBool}
  User.bIsFrontEndPreview {type=NiagaraBool}
  User.bIsInWater {type=NiagaraBool}
  User.DesertBoulder {type=NiagaraBool}
  User.SmallerSize {type=NiagaraFloat, defaultValue=0, description=""}
 ]
}
